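# Translucent, Default Lit test material for Niagara sprites that is set up to display a particle dynamic parameter as debug digits.
# The R channel of a Dynamic Parameter is copied into all four inputs of AppendMany and fed to DebugFloat4Values.
# A dot product with a constant 1 collapses the color-coded output to a scalar.
# The graph shows no connection from that result to a material output.
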Material /NiagaraStaging/TestAssets/Debug
graph {
 properties {bUsedWithNiagaraSprites=true, blendMode=BLEND_Translucent, pixelDepthOffsetMode=PDOM_Legacy, usedShadingModels=Default Lit}
 nodes [
  n1 = DynamicParameter()
  DotProduct(
   A=MaterialFunctionCall!(
    Vector4=MaterialFunctionCall(R=@n1.R, G=@n1.R, B=@n1.R, A=@n1.R) {materialFunction=&AppendMany}.RGBA
    MaximumNumberOfDigits
    UVs
    DebugTextLocation RG_UpperRight BA_LowerLeft
    Component Spacing
   ) {materialFunction=/Engine/Functions/Engine_MaterialFunctions02/Utility/DebugFloat4Values}.ColorCodedOutput
   B=Constant() {r=1}
  )
 ]
}
