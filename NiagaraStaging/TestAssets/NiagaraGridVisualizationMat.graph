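# Additive test material for Niagara mesh particles that appears to visualize grid cells as edge lines of world-space width.
# Local position scaled by particle axis lengths feeds a component-wise sphere mask, inset by WS Grid Size.
# Inverted normal-derived face masks drop each face's own axis; a max of only R and B drives emissive adjustments.
# Separate branches hold UV-border dither, normal and color experiments.
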Material /NiagaraStaging/TestAssets/NiagaraGridVisualizationMat
graph {
 properties {bUsedWithNiagaraMeshParticles=true, blendMode=BLEND_Additive, pixelDepthOffsetMode=PDOM_Legacy, usedShadingModels=Default Lit}
 nodes [
  n1 = Constant3Vector() {constant=<rgba 1 0 0 0>}
  n2 = VertexInterpolator(VS=PreSkinnedNormal())
  n3 = Multiply(A=@n1, B=Abs(None=DotProduct(A=@n2.PS, B=@n1)))
  n4 = Constant3Vector() {constant=<rgba 0 1 0 0>}
  n5 = Multiply(A=Abs(None=DotProduct(A=@n2.PS, B=@n4)), B=@n4)
  n6 = Constant3Vector() {constant=<rgba 0 0 1 0>}
  n7 = Abs(None=DotProduct(A=@n2.PS, B=@n6))
  n8 = OneMinus(None=Add(A=Add(A=@n3, B=@n5), B=Multiply(A=@n7, B=@n6)) {desc=RGB faces}) {desc=Inverse}
  n9 = Multiply(A=VertexInterpolator(VS=PreSkinnedPosition()).PS) {constB=2}
  n10 = Transform(None=Constant3Vector() {constant=<rgba 1 0 0 0>}) {transformSourceType=TRANSFORMSOURCE_ParticleWorld}
  n11 = Transform(None=Constant3Vector() {constant=<rgba 0 1 0 0>}) {transformSourceType=TRANSFORMSOURCE_ParticleWorld}
  n12 = Constant3Vector() {constant=<rgba 0 0 1 0>}
  n13 = AppendVector(
   A=AppendVector(
    A=MaterialFunctionCall(Vector 3=@n10) {materialFunction=/Engine/Functions/Engine_MaterialFunctions02/Utility/VectorLength}.V3 Length
    B=MaterialFunctionCall(Vector 3=@n11) {materialFunction=&VectorLength}.V3 Length
   )
   B=MaterialFunctionCall(Vector 3=Transform(None=@n12) {transformSourceType=TRANSFORMSOURCE_ParticleWorld}) {materialFunction=&VectorLength}.V3 Length
  )
  n14 = MaterialFunctionCall!(
   A=Abs(None=Multiply(A=@n9, B=@n13) {desc=local position scaled . .5 center})
   B=Constant()
   Radius=Subtract(A=@n13, B=ScalarParameter() {defaultValue=5, parameterName=WS Grid Size})
   Hardness=Constant() {r=100}
  ) {desc=now the elements have to be separated, materialFunction=/NiagaraStaging/ComponentWise_SphereMask}
  n15 = MaterialFunctionCall(Float3=Multiply(A=Saturate(None=@n8), B=OneMinus(None=@n14.Result))) {materialFunction=&BreakOutFloat3Components}
  n16 = Max(A=Max(A=@n15.R, B=@n15.R), B=@n15.B)
  MaterialFunctionCall(Emissive=Multiply(A=Saturate(None=@n16), B=ScalarParameter() {defaultValue=0.0214, parameterName=Param_3})) {materialFunction=&MF_FX_EmissiveAdjustments}
  n17 = TextureCoordinate()
  n18 = MaterialFunctionCall(Float2=Abs(None=ConstantBiasScale(None=@n17) {bias=-0.5, scale=2})) {materialFunction=&BreakOutFloat2Components}
  n19 = ScalarParameter() {defaultValue=0.0214, parameterName=Param_2}
  n20 = Multiply(A=Subtract(A=Max(A=@n18.R, B=@n18.R), B=OneMinus(None=@n19)) {constB=0.95}, B=Divide(B=@n19)) {constB=20}
  MaterialFunctionCall(Alpha Threshold=@n20) {materialFunction=&DitherTemporalAA}
  n21 = Subtract(A=TextureCoordinate()) {constB=0.5}
  n22 = MaterialFunctionCall(R=ScalarParameter() {defaultValue=2}, G=ScalarParameter() {defaultValue=2, parameterName=Param_1}) {materialFunction=&AppendMany}
  Normalize(VectorInput=AppendVector(A=Multiply(A=@n21, B=@n22.RG), B=Constant() {r=1}))
  Saturate(None=@n20)
  n23 = Abs(None=@n16)
  Saturate(None=Ceil(None=Multiply(A=@n23) {constB=20}))
  TextureSample(UVs=MaterialFunctionCall(Texture Scale=Constant() {r=2}) {materialFunction=&ScaleUVsByCenter}.UVs) {samplerType=SAMPLERTYPE_Normal, texture=/Engine/Functions/Engine_MaterialFunctions02/ExampleContent/Textures/SphereRenderNormalsMap}
  VectorParameter() {parameterName=Color, defaultValue=<rgba 0.4271 0.3872 0.3292 1>}
 ]
}
